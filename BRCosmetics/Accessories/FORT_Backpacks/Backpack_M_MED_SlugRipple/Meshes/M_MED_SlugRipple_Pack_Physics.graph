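# Physics asset for the SlugRipple backpack cosmetic, covering its dangling gun and mace bones.
# Small sphere bodies sit on dyn_gun, dyn_mace and a "basa" (base) bone for each.
# Two constraints pair each prop with its base: the gun locks both swings but not twist, while the mace locks swing and twist.
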PhysicsAsset /BRCosmetics/Accessories/FORT_Backpacks/Backpack_M_MED_SlugRipple/Meshes/M_MED_SlugRipple_Pack_Physics
bodies {
 dyn_gun [{shapeType=Sphere, radius=3, shapeName=dyn_gun_sphere}]
 dyn_gun_basa [{shapeType=Sphere, radius=1, shapeName=dyn_gun_basa_sphere}]
 dyn_mace [{shapeType=Sphere, radius=2, shapeName=dyn_mace_sphere}]
 dyn_mace_basa [{shapeType=Sphere, radius=1, shapeName=dyn_mace_basa_sphere}]
}
constraints [
 {
  bone1Name=dyn_gun
  bone2Name=dyn_gun_basa
  swing1LimitDegrees=45
  swing1Motion=Locked
  swing2LimitDegrees=45
  swing2Motion=Locked
  twistLimitDegrees=16.32
 }
 {
  bone1Name=dyn_mace
  bone2Name=dyn_mace_basa
  swing1LimitDegrees=45
  swing1Motion=Locked
  swing2LimitDegrees=45
  swing2Motion=Locked
  twistLimitDegrees=45
  twistMotion=Locked
 }
]
